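# Sound cue for the landing foley of the WartyBark quadruped cosmetic companion.
# A quality-level root sends its first two slots to an equally weighted random pick of seven Land waves, and its last two to a pick of the first five.
# Output is slightly pitched down and reduced in volume, using companion foley attenuation, the pets sound class and the shared quadruped footstep concurrency.
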SoundCue /CosmeticCompanions/Assets/Quadruped/WartyBark/Audio/Foley/SC_MImosa_Quadraped_WartyBark_Land
root=QualityLevel_0
nodes [
 {id=QualityLevel_0, children [Random_2,Random_2,Random_1,Random_1]}
 {id=Random_1, children [WavePlayer_22, WavePlayer_23, WavePlayer_24, WavePlayer_25, WavePlayer_26], properties {weights [1,1,1,1,1]}}
 {
  id=Random_2
  children [WavePlayer_15, WavePlayer_16, WavePlayer_17, WavePlayer_18, WavePlayer_19, WavePlayer_20, WavePlayer_21]
  properties {weights [1,1,1,1,1,1,1]}
 }
 {
  id=WavePlayer_15
  properties {soundWaveAssetPtr=/CosmeticCompanions/Assets/Quadruped/WartyBark/Audio/Foley/Wavs/Mimosa_Quadraped_WartyBark_Land_01}
 }
 {
  id=WavePlayer_16
  properties {soundWaveAssetPtr=/CosmeticCompanions/Assets/Quadruped/WartyBark/Audio/Foley/Wavs/Mimosa_Quadraped_WartyBark_Land_02}
 }
 {
  id=WavePlayer_17
  properties {soundWaveAssetPtr=/CosmeticCompanions/Assets/Quadruped/WartyBark/Audio/Foley/Wavs/Mimosa_Quadraped_WartyBark_Land_03}
 }
 {
  id=WavePlayer_18
  properties {soundWaveAssetPtr=/CosmeticCompanions/Assets/Quadruped/WartyBark/Audio/Foley/Wavs/Mimosa_Quadraped_WartyBark_Land_04}
 }
 {
  id=WavePlayer_19
  properties {soundWaveAssetPtr=/CosmeticCompanions/Assets/Quadruped/WartyBark/Audio/Foley/Wavs/Mimosa_Quadraped_WartyBark_Land_05}
 }
 {
  id=WavePlayer_20
  properties {soundWaveAssetPtr=/CosmeticCompanions/Assets/Quadruped/WartyBark/Audio/Foley/Wavs/Mimosa_Quadraped_WartyBark_Land_06}
 }
 {
  id=WavePlayer_21
  properties {soundWaveAssetPtr=/CosmeticCompanions/Assets/Quadruped/WartyBark/Audio/Foley/Wavs/Mimosa_Quadraped_WartyBark_Land_07}
 }
 {
  id=WavePlayer_22
  properties {soundWaveAssetPtr=&Mimosa_Quadraped_WartyBark_Land_01}
 }
 {
  id=WavePlayer_23
  properties {soundWaveAssetPtr=&Mimosa_Quadraped_WartyBark_Land_02}
 }
 {
  id=WavePlayer_24
  properties {soundWaveAssetPtr=&Mimosa_Quadraped_WartyBark_Land_03}
 }
 {
  id=WavePlayer_25
  properties {soundWaveAssetPtr=&Mimosa_Quadraped_WartyBark_Land_04}
 }
 {
  id=WavePlayer_26
  properties {soundWaveAssetPtr=&Mimosa_Quadraped_WartyBark_Land_05}
 }
]
properties {
 attenuationSettings=/Game/Sounds/Attenuation/AS_Companion_Foley
 maxDistance=1200
 pitchMultiplier=0.9
 priority=0
 soundClassObject=&Fort_Creatures_Pets
 volumeMultiplier=0.45
 concurrencySet [/CosmeticCompanions/Assets/Audio_Shared/Foley/Quadraped/CONC_Mimosa_Quadraped_Footstep]
}
