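# Sound Cue for short slide foley on the WartyBark quadruped cosmetic companion.
# A Random node picks one of five equally weighted SlideShort wave variations.
# The cue plays quiet and pitched up in the creature pets sound class, using companion foley attenuation and sharing the quadruped footstep concurrency set.
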SoundCue /CosmeticCompanions/Assets/Quadruped/WartyBark/Audio/Foley/SC_MImosa_Quadraped_WartyBark_SlideShort
root=Random_4
nodes [
 {id=Random_4, children [WavePlayer_14, WavePlayer_15, WavePlayer_16, WavePlayer_17, WavePlayer_18], properties {weights [1,1,1,1,1]}}
 {
  id=WavePlayer_14
  properties {soundWaveAssetPtr=/CosmeticCompanions/Assets/Quadruped/WartyBark/Audio/Foley/Wavs/Mimosa_Quadraped_WartyBark_SlideShort_01}
 }
 {
  id=WavePlayer_15
  properties {soundWaveAssetPtr=/CosmeticCompanions/Assets/Quadruped/WartyBark/Audio/Foley/Wavs/Mimosa_Quadraped_WartyBark_SlideShort_02}
 }
 {
  id=WavePlayer_16
  properties {soundWaveAssetPtr=/CosmeticCompanions/Assets/Quadruped/WartyBark/Audio/Foley/Wavs/Mimosa_Quadraped_WartyBark_SlideShort_03}
 }
 {
  id=WavePlayer_17
  properties {soundWaveAssetPtr=/CosmeticCompanions/Assets/Quadruped/WartyBark/Audio/Foley/Wavs/Mimosa_Quadraped_WartyBark_SlideShort_04}
 }
 {
  id=WavePlayer_18
  properties {soundWaveAssetPtr=/CosmeticCompanions/Assets/Quadruped/WartyBark/Audio/Foley/Wavs/Mimosa_Quadraped_WartyBark_SlideShort_05}
 }
]
properties {
 attenuationSettings=/Game/Sounds/Attenuation/AS_Companion_Foley
 maxDistance=1200
 pitchMultiplier=1.2
 priority=0
 soundClassObject=&Fort_Creatures_Pets
 volumeMultiplier=0.3
 concurrencySet [/CosmeticCompanions/Assets/Audio_Shared/Foley/Quadraped/CONC_Mimosa_Quadraped_Footstep]
}
